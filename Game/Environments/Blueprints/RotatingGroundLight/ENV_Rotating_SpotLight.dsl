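; Fortnite ground spotlight building actor that poses a light fixture and its light during construction; despite the name, nothing rotates at runtime.
; Mesh sockets position the fixture on the base and the first light component on the fixture.
; The fixture angle, clamped to ±45 degrees, turns on an enum-selected axis, with an optional light angle override combined with it.
; The fixture inherits the base mesh material and the building texture data's tint.
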(event EventBeginPlay
  (|CallParentFunction) ; This node is disabled and will not be called.\nDrag off pins to build functionality.
)

(event Collision|EventActorBeginOverlap
  (params (OtherActor :type Actor%20Object%20Reference)))

(event EventTick
  (params (DeltaSeconds :type Float%20%28single-precision%29)))

(fn GetFixturePosition
  (bind _staticmeshcomponent (Variables|Debug|GetStaticMeshComponent))
  (bind _returnvalue (Transformation|GetAllSocketNames _staticmeshcomponent))
  (if (Utilities|Array|IsValidIndex _returnvalue)
    (Variables|GroundLight|SetFixtureLocation (.location (Transformation|GetSocketTransform _staticmeshcomponent (Utilities|Array|Get(acopy) _returnvalue) "RTS_Actor")))
    (else
      (Development|PrintString "Cannot find socket on static mesh"))))

(fn GetLightPosition
  (if (Utilities|Array|IsValidIndex (Transformation|GetAllSocketNames (Variables|GroundLight|GetLightFixture)))
    (Variables|GroundLight|SetLightPosition (Transformation|GetSocketLocation (Variables|GroundLight|GetLightFixture) (Utilities|Array|Get(acopy) (Transformation|GetAllSocketNames (Variables|GroundLight|GetLightFixture)))))
    (else
      (Development|PrintString "Fixture mesh does not contain a valid socket"))))

(fn GetMaterialFromRootMesh
  (Rendering|Material|SetMaterial (Variables|GroundLight|GetLightFixture) 0 (Material|GetMID (Variables|Debug|GetStaticMeshComponent))))

(fn GetTintColorFromTextureData
  (bind _outtexturedataarray (Building|GetBuildingTextureData))
  (if (> (Utilities|Array|Length _outtexturedataarray) 0)
    (Variables|GroundLight|SetTintColor (Math|Conversions|ToLinearColor(Color) (Variables|/Script/FortniteGame.BuildingTextureData|GetTintColor (Utilities|Array|Get(acopy) _outtexturedataarray))))))

(fn SetFixturePosition
  (if (> (Utilities|Array|Length (Transformation|GetAllSocketNames (Variables|Debug|GetStaticMeshComponent))) 0)
    (Utilities|IsValid (Variables|GroundLight|GetLightFixture)
      (:Is%20Valid
        (Transformation|SetRelativeLocation (Variables|GroundLight|GetLightFixture) (Variables|GroundLight|GetFixtureLocation) false false)
        (switch Utilities|FlowControl|Switch|SwitchonENUM_RotationAxis (Variables|GroundLight|GetENUM_RotationAxis)
          (:NewEnumerator0
            (CallFunction|SetFixtureRotationAngle)
            (Transformation|SetRelativeRotation (Variables|GroundLight|GetLightFixture) (Math|Rotator|MakeRotator (Variables|GroundLight|GetFixtureAngle)) false false))
          (:NewEnumerator1
            (CallFunction|SetFixtureRotationAngle)
            (Transformation|SetRelativeRotation (Variables|GroundLight|GetLightFixture) (Math|Rotator|MakeRotator 0.0 (Variables|GroundLight|GetFixtureAngle)) false false))
          (:NewEnumerator2
            (CallFunction|SetFixtureRotationAngle)
            (Transformation|SetRelativeRotation (Variables|GroundLight|GetLightFixture) (Math|Rotator|MakeRotator 0.0 0.0 (Variables|GroundLight|GetFixtureAngle)) false false))))
      (:Is%20Not%20Valid
        (Development|PrintString "Assign a static mesh to Light fixture component")))
    (else
      (Development|PrintString "No valid socket found on static mesh"))))

(fn SetFixtureRotationAngle
  (bind _fixtureangle (Variables|GroundLight|GetFixtureAngle))
  (if (> _fixtureangle 45.0)
    (Variables|GroundLight|SetFixtureAngle 45.0)
    (elif (< _fixtureangle -45.0)
      (Variables|GroundLight|SetFixtureAngle -45.0))))

(fn SetLightComponentAngle
  (if (Variables|GroundLight|GetOverrideLightAngle)
    (if (Utilities|IsValid (Variables|GroundLight|GetLightComponent) :!class /Script/Engine.KismetSystemLibrary)
      (Transformation|SetRelativeRotation (Variables|GroundLight|GetLightComponent) (Math|Rotator|CombineRotators (Variables|GroundLight|GetLightComponentAngle) (Variables|/Script/Engine.SceneComponent|GetRelativeRotation (Variables|GroundLight|GetLightFixture))) false false))))

(fn SetLightPosition
  (bind _returnvalue (Actor|GetComponentsByClass "/Script/Engine.LightComponent"))
  (if (Utilities|Array|IsValidIndex _returnvalue)
    (bind _output_get (Variables|GroundLight|SetLightComponent (Utilities|Array|Get(acopy) _returnvalue)))
    (Transformation|SetWorldLocation _output_get (Variables|GroundLight|GetLightPosition) false false)
    (else
      (Development|PrintString "No valid light component found"))))

(fn SetTintColorOnMeshParts
  (Default|SetColoronSecondaryMesh (Variables|GroundLight|GetTintColor) (Variables|GroundLight|GetLightFixture)))

(fn ConstructionScript
  (|CallParentFunction)
  (CallFunction|GetFixturePosition)
  (CallFunction|SetFixturePosition)
  (CallFunction|GetLightPosition)
  (CallFunction|SetLightPosition)
  (CallFunction|SetLightComponentAngle)
  (Default|GetMaterialfromRootMesh)
  (Default|GetTintColorfromTextureData)
  (Default|SetTintColoronMeshParts))
